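; Weapon cosmetics Blueprint for the Ancient Gladiator pickaxe that tells its idle VFX whether it is shown in the front end.
; On weapon cosmetics init, it runs the parent logic, casts the game state to FortGameStateFrontEnd and sets bisFE to true.
; It then passes that flag to the idle VFX component as the bisFrontEnd bool parameter.
; BeginPlay and Tick are empty, and the construction script only calls the parent.
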(event EventOnInitWeaponCosmetics
  (|CallParentFunction)
  (Utilities|Casting|CastToFortGameStateFrontEnd (Game|GetGameState))
  (Variables|Default|SetbisFE true)
  (Effects|Components|ParticleSystem|SetBoolParameter (CallFunction|GetIdleVFXComponent) "bisFrontEnd" (Variables|Default|GetbisFE)))

(event EventBeginPlay)

(event EventTick
  (params (DeltaSeconds :type Float%20%28single-precision%29)))

(fn ConstructionScript
  (|CallParentFunction))
